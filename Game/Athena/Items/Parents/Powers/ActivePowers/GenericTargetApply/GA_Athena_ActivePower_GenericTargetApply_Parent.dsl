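; Parent active power ability for Athena that runs target selection from a cast montage notify and applies gameplay effects to the targets.
; The tagged notify passes a do-once gate, reset on ability init, and an optional commit check before targeting.
; Targeting plays a burst gameplay cue, then applies the tagged effect container to the target data.
; Montage blend-out, cancellation, end notify or commit failure ends the ability.
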(event EventCastMontageBlendOut
  ;[1 1 1] MontageBlendOut - End Ability
  (|CallParentFunction)
  (Ability|EndAbility))

(event EventCastMontageNotifyEndAbility
  ;[1 1 1] MontageEndNotify - End Ability
  (|CallParentFunction)
  (Ability|EndAbility))

(event EventCastMontageCancelledOrInterrupted
  ;[1 1 1] Montage Cancelled or Interrupted - end ability
  (|CallParentFunction)
  (Ability|EndAbility))

(event EventPowerCommitFailed
  ;[1 1 1] Commit Failed - End Ability
  (|CallParentFunction)
  (Ability|EndAbility))

(event EventCastMontageNotifyStart
  (params (NotifyTag :type Gameplay%20Tag%20Structure))
  ;[1 1 1] Event Montage Notify Start Do the commit then target. Check if we should commit for every target or if we should ignore the commit.
  (|CallParentFunction NotifyTag)
  (if (GameplayTags|Equal(GameplayTag) NotifyTag (Variables|PowerTargetApply|GetTag_TargetApplyNotify))
    (Utilities|DoOnceThenDo false
      (:DoOnce
        _)
      (:ThenDo
        (if (Variables|PowerAbilityCommit|GetCommitOnEveryTargeting)
          (Utilities|PowerCommitAbility (Variables|AbilityCommit|GetIgnoreAbilityCommit)
            (:Pass
              _)
            (:Fail
              (Development|PrintString "Power: Ability Commit Failed")
              (CallFunction|PowerCommitFailed))
            (:Ignored
              _))
          (else
            _))
        (CallFunction|Target)))
    (elif (and (GameplayTags|Equal(GameplayTag) NotifyTag (Variables|PowerAnimationCast|GetT_AbilityEndCastNotify)) (GameplayTags|NotEqual(GameplayTag) :A (Variables|PowerAnimationCast|GetT_AbilityEndCastNotify)))
      (CallFunction|CastMontageNotifyEndAbility))))

(event Custom|Target%20
  ;[1 1 1] Do Targeting And Apply Effect - Play burst GC, do target selection
  (CallFunction|PlayBurstGC (Variables|PowerGameplayCues|GetGC_Targeting) (Variables|PowerGameplayCues|GetTargetingGCUsesParams) (CallFunction|TargetingGCParams))
  (bind (_asynctaskproxy _targetdata _applicationtag) (Ability|Tasks|DoTargetSelection "Targeting" (Variables|PowerTargetApply|GetTag_TargetingApplication) true)
    (:Targeted
      (CallFunction|ApplyEffect _targetdata _applicationtag))))

(event Custom|ResetCommitDoOnce
  ;[1 1 1] Event Montage Notify Start Do the commit then target. Check if we should commit for every target or if we should ignore the commit.
  (Utilities|DoOnceThenDo false
    (:DoOnce
      _)
    (:ThenDo
      (if (Variables|PowerAbilityCommit|GetCommitOnEveryTargeting)
        (Utilities|PowerCommitAbility (Variables|AbilityCommit|GetIgnoreAbilityCommit)
          (:Pass
            _)
          (:Fail
            (Development|PrintString "Power: Ability Commit Failed")
            (CallFunction|PowerCommitFailed))
          (:Ignored
            _))
        (else
          _))
      (CallFunction|Target))))

(event Custom|ApplyEffect
  (params (TargetData :type Gameplay%20Ability%20Target%20Data%20Handle%20Structure) (ApplicationTag :type Gameplay%20Tag%20Structure))
  ;[1 1 1] Apply Effect to Target Data
  (Ability|ApplyGameplayEffectContainer TargetData ApplicationTag (Variables|PowerTargetApply|GetApplyEffectLevel)))

(fn InitAbility
  ;[1 1 1] On Init Reset the Do Onces.
  (|CallParentFunction)
  (CallFunction|ResetCommitDoOnce))

(fn TargetingGCParams
  (return 0))
